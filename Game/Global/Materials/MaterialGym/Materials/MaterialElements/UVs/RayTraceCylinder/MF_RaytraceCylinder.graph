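# Material function that analytically ray-traces an uncapped cylinder per pixel from origin, radius and axis inputs.
# It moves the camera ray into an orthonormal basis built from the axes, then solves a quadratic in the XY plane for entry and exit hits.
# The hits yield front and back local positions, cylindrical atan2 UVs and world normals with vertex-normal fallback, plus thickness and a hit mask.
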MaterialFunction /Game/Global/Materials/MaterialGym/Materials/MaterialElements/UVs/RayTraceCylinder/MF_RaytraceCylinder
graph {
 properties {
  bExposeToLibrary=true
  description=Returns an analytical intersection for a cylinder of the specified coordinates.
  libraryCategoriesText [Math]
 }
 nodes [
  FunctionOutput(None=NamedRerouteUsage()) {outputName=Frontface UV, sortPriority=2}
  FunctionOutput(None=NamedRerouteUsage()) {outputName=Frontface Pos LS}
  FunctionOutput(None=NamedRerouteUsage()) {outputName=Mask, sortPriority=7}
  FunctionOutput(None=NamedRerouteUsage()) {outputName=Backface Pos, sortPriority=3}
  FunctionOutput(None=NamedRerouteUsage()) {outputName=Thickness, sortPriority=6}
  FunctionOutput(None=NamedRerouteUsage()) {outputName=Backface UV, sortPriority=5}
  FunctionOutput(None=NamedRerouteUsage()) {outputName=Frontface Normal WS, sortPriority=1}
  FunctionOutput(None=NamedRerouteUsage()) {outputName=Backface Normal WS, sortPriority=4}
  n1 = FunctionInput(Preview=Transform(None=Constant3Vector() {constant=<rgba 0 0 1 0>}) {transformSourceType=TRANSFORMSOURCE_Local}) {
   bUsePreviewValueAsDefault=true
   description=The Z axis is the axis along the length of the cylinder. Defaults to the local Z axis of the mesh the material is applied to.
   inputName=Z Axis
   sortPriority=3
  }
  NamedRerouteDeclaration(None=Normalize(VectorInput=@n1)) {name=Z_Axis, nodeColor=<rgba 0 1 0.6719 1>}
  n2 = FunctionInput(Preview=Transform(None=Constant3Vector() {constant=<rgba 1 0 0 0>}) {transformSourceType=TRANSFORMSOURCE_Local}) {
   bUsePreviewValueAsDefault=true
   description="The X axis is the axis used to apply non uniform scale along.  Defaults to the local X axis of the mesh the material is applied to."
   inputName=X Axis
   sortPriority=4
  }
  NamedRerouteDeclaration(None=Normalize(VectorInput=@n2)) {name=X_Axis, nodeColor=<rgba 1 0.3516 0 1>}
  NamedRerouteDeclaration(None=MaterialFunctionCall!(VectorToTransform=NamedRerouteUsage(), BasisX=NamedRerouteUsage(), BasisY=NamedRerouteUsage(), BasisZ=NamedRerouteUsage()) {desc=Directional Vec, materialFunction=/Engine/Functions/Engine_MaterialFunctions02/Math/InverseTransformMatrix}.Transformed Vector 3) {name=TmCameraRay, nodeColor=<rgba 0 0.5312 1 1>}
  n3 = CameraPositionWS()
  n4 = WorldPosition()
  NamedRerouteDeclaration(None=Normalize(VectorInput=Subtract(A=@n3, B=@n4.XYZ))) {name=PosToCam, nodeColor=<rgba 0.8281 0 1 1>}
  n5 = MaterialFunctionCall!(Vector1=NamedRerouteUsage(), Vector2=NamedRerouteUsage()) {materialFunction=/Engine/Functions/Engine_MaterialFunctions02/Math/CreateThirdOrthogonalVector}
  NamedRerouteDeclaration(None=@n5.Vector1) {name=BasisZ, nodeColor=<rgba 0 0.0275 1 1>}
  n6 = NamedRerouteUsage()
  n7 = NamedRerouteUsage()
  NamedRerouteDeclaration(None=Add(A=NamedRerouteUsage(), B=Multiply(A=@n6, B=@n7)) {desc="Ray Entry, Local"}) {name=LS_RayEntry, nodeColor=<rgba 0 0.0625 1 1>}
  NamedRerouteDeclaration(None=Subtract(A=NamedRerouteUsage(), B=NamedRerouteUsage()) {desc=Uncapped Cylinder Thickness}) {name=Thickness, nodeColor=<rgba 1 0 0.9609 1>}
  NamedRerouteDeclaration(None=@n5.Vector2) {name=BasisX, nodeColor=<rgba 0.8333 0 0.0256 1>}
  n8 = MaterialFunctionCall(
   VectorToTransform=Multiply(A=NamedRerouteUsage(), B=Constant3Vector() {constant=<rgba 1 1 0 0>})
   BasisX=NamedRerouteUsage()
   BasisY=NamedRerouteUsage()
   BasisZ=NamedRerouteUsage()
  ) {materialFunction=&Transform3x3Matrix}
  n9 = VertexNormalWS()
  NamedRerouteDeclaration(None=If(A=NamedRerouteUsage(), "A > B"=Normalize(VectorInput=@n8.Transformed Vector), "A == B"=@n9, "A < B"=@n9) {constB=0.5}) {name=NormalFrontface, nodeColor=<rgba 0 1 0.7188 1>}
  n10 = Constant()
  NamedRerouteDeclaration(None=If(A=NamedRerouteUsage(), B=@n10, "A > B"=Constant() {r=1}, "A == B"=@n10, "A < B"=Constant())) {name=Mask, nodeColor=<rgba 0.0781 0 1 1>}
  n11 = NamedRerouteUsage()
  n12 = NamedRerouteUsage()
  NamedRerouteDeclaration(None=Add(A=NamedRerouteUsage(), B=Multiply(A=@n11, B=@n12)) {desc="Ray Entry, Local"}) {name=LS_RayExit, nodeColor=<rgba 0 0.0625 1 1>}
  n13 = NamedRerouteUsage()
  MaterialExpressionNormalize_3 = Normalize(VectorInput=@n13)
  n14 = Convert(None=@MaterialExpressionNormalize_3) {
   nodeName=Break Float3
   convertInputs [{
     type=Vector3
     expressionInput {
      expression=&self:MaterialExpressionNormalize_3
      inputName=None
      mask=0
      maskA=0
      maskB=0
      maskG=0
      maskR=0
      outputIndex=0
     }
    }]
   convertOutputs [{type=Scalar}, {type=Scalar}, {type=Scalar}]
  }
  n15 = Divide(A=Arctangent2Fast(Y=@n14, X=@n14), B=Multiply(A=NamedRerouteUsage()) {constB=-1}) {constB=-6.28}
  n16 = Constant() {r=2}
  n17 = Multiply(
   A=NamedRerouteDeclaration(None=MaterialFunctionCall(Multiplier=@n16) {materialFunction=&Pi}.Result) {name=2PI, nodeColor=<rgba 0.7344 1 0 1>}
   B=NamedRerouteUsage()
  )
  n18 = Divide(
   A=ComponentMask(None=@n13) {b=true}
   B=NamedRerouteDeclaration(None=Multiply(A=@n17) {constB=-1}) {name=circumference, nodeColor=<rgba 0 1 0.625 1>}
  )
  NamedRerouteDeclaration(None=Add(A=AppendVector(A=@n15, B=@n18)) {constB=0.5}) {name=FrontFaceUV, nodeColor=<rgba 0.9453 0 1 1>}
  NamedRerouteDeclaration(None=@n5.Vector3) {name=BasisY, nodeColor=<rgba 0.1128 0.8177 0 1>}
  n19 = NamedRerouteUsage()
  MaterialExpressionNormalize_4 = Normalize(VectorInput=@n19)
  n20 = Convert(None=@MaterialExpressionNormalize_4) {
   nodeName=Break Float3
   convertInputs [{
     type=Vector3
     expressionInput {
      expression=&self:MaterialExpressionNormalize_4
      inputName=None
      mask=0
      maskA=0
      maskB=0
      maskG=0
      maskR=0
      outputIndex=0
     }
    }]
   convertOutputs [{type=Scalar}, {type=Scalar}, {type=Scalar}]
  }
  n21 = Divide(A=Arctangent2Fast(Y=@n20, X=@n20), B=NamedRerouteUsage()) {constB=-6.28}
  n22 = Divide(A=ComponentMask(None=@n19) {b=true}, B=NamedRerouteUsage())
  NamedRerouteDeclaration(None=Add(A=AppendVector(A=@n21, B=@n22)) {constB=0.5}) {name=BackFaceUV, nodeColor=<rgba 0.9453 0 1 1>}
  n23 = NamedRerouteUsage()
  n24 = Constant3Vector() {constant=<rgba 1 1 0 0>}
  n25 = Normalize(VectorInput=MaterialFunctionCall(VectorToTransform=Multiply(A=@n23, B=@n24), BasisX=NamedRerouteUsage(), BasisY=NamedRerouteUsage(), BasisZ=NamedRerouteUsage()) {materialFunction=&Transform3x3Matrix}.Transformed Vector)
  n26 = VertexNormalWS()
  NamedRerouteDeclaration(None=If(A=NamedRerouteUsage(), "A > B"=Multiply(A=@n25) {constB=-1}, "A == B"=@n26, "A < B"=@n26) {constB=0.5}) {name=NormalBackface, nodeColor=<rgba 0 1 0.7188 1>}
  n27 = ActorPositionWS()
  NamedRerouteDeclaration(None=Subtract(
    A=CameraPositionWS()
    B=FunctionInput(Preview=@n27) {bUsePreviewValueAsDefault=true, description=The center of the cylinder. Defaults to Actor Position., inputName=Origin}
   )) {name=DirToCenter, nodeColor=<rgba 0 0.9297 1 1>}
  NamedRerouteDeclaration(None=MaterialFunctionCall(VectorToTransform=NamedRerouteUsage(), BasisX=NamedRerouteUsage(), BasisY=NamedRerouteUsage(), BasisZ=NamedRerouteUsage()) {desc=Origin Vec, materialFunction=&InverseTransformMatrix}.Transformed Vector 3) {name=TmDirToCenter, nodeColor=<rgba 1 0.75 0 1>}
  n28 = MaterialFunctionCall(Float2=NamedRerouteUsage()) {materialFunction=&BreakOutFloat2Components}
  n29 = MaterialFunctionCall(Float2=NamedRerouteUsage()) {materialFunction=&BreakOutFloat2Components}
  n30 = MaterialFunctionCall(Float2=NamedRerouteUsage()) {materialFunction=&BreakOutFloat2Components}
  n31 = Multiply(A=@n29.R, B=@n30.R)
  n32 = Multiply(A=@n29.G, B=@n30.G)
  n33 = MaterialFunctionCall(Float2=NamedRerouteUsage()) {materialFunction=&BreakOutFloat2Components}
  n34 = Multiply(A=@n33.R, B=@n33.R)
  n35 = Multiply(A=@n33.G, B=@n33.G)
  n36 = FunctionInput() {
   bUsePreviewValueAsDefault=true
   description="The X axis is the axis used to apply non uniform scale along.  Defaults to the local X axis of the mesh the material is applied to."
   inputName=Radius
   inputType=FunctionInput_Scalar
   sortPriority=1
   previewValue=<50 0 0 1>
  }
  n37 = NamedRerouteDeclaration(None=@n36) {name=radius, nodeColor=<rgba 0 0.4141 1 1>}
  n38 = MaterialFunctionCall!(
   A=Add(A=Multiply(A=@n28.R, B=@n28.R) {desc="D.x ^2"}, B=Multiply(A=@n28.G, B=@n28.G) {desc="A D.y ^2"}) {desc=A}
   B=Add(A=Multiply(A=@n31) {constB=2}, B=Multiply(A=@n32) {constB=2}) {desc=B}
   C=Subtract(A=Add(A=@n34, B=@n35), B=Multiply(A=@n37, B=@n37)) {desc=C}
  ) {materialFunction=/Engine/Functions/Engine_MaterialFunctions01/Math/QuadraticFormula}
  NamedRerouteDeclaration(None=@n38.T0) {name=T1, nodeColor=<rgba 0.4297 0 1 1>}
  NamedRerouteDeclaration(None=@n38.T1) {name=T0, nodeColor=<rgba 0 1 0.1094 1>}
  NamedRerouteDeclaration(None=@n38.Discriminant) {name=Discriminant, nodeColor=<rgba 1 0 0.2109 1>}
 ]
}
